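; Ashton Pack "Milo" assault weapon Blueprint that drives overheat VFX and audio from the weapon's heat events.
; While heating, a looping timer re-activates the overheat Niagara effect with reset, at a longer interval on mobile; cooling start clears the timer.
; Overheating fires the effect, spawns an attached sound, then immediately deactivates the effect.
; Finishing cooling or equipping deactivates the effect, the latter so it does not stay stuck on.
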(event FortWeapon|EventOnEquippedAndAttached
  ;[1 1 1] On Equip -> deactivate FX or else they get stuck on if previously heated up
  (|CallParentFunction)
  (Components|Activation|Deactivate (Variables|Default|GetNS_Ashton_Milo_Overheat)))

(event Overheating|EventOnStartOverheated
  ;[1 1 1] Overheated
  (|CallParentFunction)
  (Components|Activation|Activate (Variables|Default|GetNS_Ashton_Milo_Overheat) true)
  (Audio|SpawnSoundAttached "/Figment_Martin/Athena/Sounds/LTMs/Ashton/Milo/Ashton_Milo_Primary_Overheat.Ashton_Milo_Primary_Overheat" (Variables|Default|GetNS_Ashton_Milo_Overheat))
  (Components|Activation|Deactivate (Variables|Default|GetNS_Ashton_Milo_Overheat)))

(event Custom|OverheatFX
  ;[1 1 1] Overheat FX
  (Components|Activation|Activate (Variables|Default|GetNS_Ashton_Milo_Overheat) true))

(event Overheating|EventOnStartHeating
  ;[1 1 1] Start Heating
  (|CallParentFunction)
  (bind _returnvalue (Utilities|Time|SetTimerbyEvent (EventDispatchers|CreateEvent :!event Custom|OverheatFX) (select (Fortnite|IsMobilePlatform) 0.1 0.03) true))
  (Variables|VFX|SetOverheatFXTimer _returnvalue))

(event Overheating|EventOnFinishedCooling
  ;[1 1 1] Finished Cooling
  (|CallParentFunction)
  (Components|Activation|Deactivate (Variables|Default|GetNS_Ashton_Milo_Overheat)))

(event Overheating|EventOnStartCooling
  ;[1 1 1] Start Cooling
  (|CallParentFunction)
  (Utilities|Time|ClearandInvalidateTimerbyHandle (Variables|VFX|GetOverheatFXTimer)))

(fn ConstructionScript
  (|CallParentFunction))
